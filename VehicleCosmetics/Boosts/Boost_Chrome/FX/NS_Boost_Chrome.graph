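# Niagara system for the Chrome vehicle boost cosmetic, trailing sprites and layered ribbons behind the car's boost ports.
# Each emitter runs a filtered socket count, and PortCount spreads sprites and ribbons across skeletal mesh sockets.
# Two ribbon emitters differ in speed, lifetime, width and color, stack several renderers, offset by game mode, and inherit source movement that fades with age.
# Painted-variant override modules are present but disabled.
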NiagaraSystem /VehicleCosmetics/Boosts/Boost_Chrome/FX/NS_Boost_Chrome
dependencies {
 usedDataInterfaces [
  &NiagaraDataInterfaceCurve
  &NiagaraDataInterfaceVector2DCurve
  &NiagaraDataInterfaceColorCurve
  &NiagaraDataInterfaceSkeletalMesh
 ]
 usedDynamicInputs [&RandomRangeFloat, &Vector2DFromCurve]
 usedModules [
  &self:SystemSpawnScript.SystemScriptSource.SystemScriptGraph.NiagaraNodeAssignment_0.SetVariables_#1
  &SystemState
 ]
}
emitters {
 NE_VC_Boost_Sprite_Parent {
  emitterUpdateScript [
   &EmitterState()
   &VC_FilteredSocketCount()
   &SpawnRate(SpawnRate=12, Use Spawn Probability=false)
  ]
  particleSpawnScript [
   &VC_PaintedVariantsOverride() {enabled=false}
   &InitializeParticle(
    Color=<rgba 0.7917 0.7917 0.7917 1>
    Sprite Rotation Angle=0.4
    Sprite Rotation Mode=::NewEnumerator1
    Sprite Size Mode=::NewEnumerator1
    Uniform Sprite Size Max=18
    Uniform Sprite Size Min=8
    UsePositionOffset=false
    Write Material Random=true
   )
   &PartitionParticles(Number Of Partitions=$Emitter.PortCount)
   &SkeletalMeshLocation(
    Bone / Socket Index=$Particles.PartitionParticles.Partition
    Bone / Socket Sampling Mode=::NewEnumerator5
    MaskSpawnGroup=false
    Mesh Sampling Type=::NewEnumerator6
    Sampled Position Offset=&AddVectorToPosition(Position=$User.DTCarOffset, Vector=<0 0 0>)
    Skeletal Mesh=$System.SkeletalMesh
   )
   &AddVelocity(Velocity=&RandomRangeVector(Minimum=<-120 0 0>, Maximum=<-400 1 1>))
   &ShapeLocation(
    MaskSpawnGroup=false
    Offset=<-20 0 0>
    Offset Mode=::NewEnumerator0
    Override Seed=false
    Shape Origin=$Particles.Position
    Sphere Radius=25
    UseOwnerScale=true
   )
  ]
  particleUpdateScript [
   &ParticleState()
   &ScaleSpriteSize(
    Scale Sprite Size Mode=::NewEnumerator0
    Uniform Scale Factor=&FloatFromCurve(FloatCurve=DataInterface() {propertyValues {Curve {
        defaultValue=50
        keys [
         {interpMode=RCIM_Cubic, tangentMode=RCTM_User, time=0, value=1}
         {interpMode=RCIM_Cubic, tangentMode=RCTM_User, time=0.1, value=2}
         {arriveTangent=-3.3333, interpMode=RCIM_Cubic, leaveTangent=-3.3333, time=0.25, value=1.5}
         {arriveTangent=16.0587, interpMode=RCIM_Cubic, leaveTangent=16.0587, tangentMode=RCTM_User, time=0.4, value=1}
         {arriveTangent=16.3709, interpMode=RCIM_Cubic, leaveTangent=16.3709, tangentMode=RCTM_User, time=0.6, value=1.5}
         {arriveTangent=-3.75, interpMode=RCIM_Cubic, leaveTangent=-3.75, time=0.75, value=1}
         {tangentMode=RCTM_User, time=1, value=0}
        ]
       }}})
   )
   &SolveForcesAndVelocity(Clamp Velocity=false, Limit Acceleration=false)
   &ScaleColor(Scale RGB=<3 3 3>, ScaleA=true, ScaleRGB=true)
  ]
  rendererClasses [&NiagaraSpriteRendererProperties]
  renderers [{rendererClass=&NiagaraSpriteRendererProperties}]
 }
 Ribbon002 {
  emitterUpdateScript [
   &EmitterState()
   &VC_FilteredSocketCount()
   &VC_RateQualityHandler(Epic Spawn Rate=35, High Spawn Rate=35, Low Spawn Rate=35, Medium Spawn Rate=35, Use Custom Rate?=true)
   &VC_SpawnRibbon_v1(Number of Ribbons=$Emitter.PortCount, SpawnRate=$Emitter.Spawn Rate)
  ]
  particleSpawnScript [
   &VC_PaintedVariantsOverride() {enabled=false}
   &VC_InitializeRibbon_v1(
    Custom Offset=true
    Enable Skeletal Position=true
    Lifetime=1.1
    Local Vector Offset=/VehicleCosmetics/FX/Niagara/Modules/GameMode/VC_SetVectorByGameModeID(
     SameValueForLockerAndBR?:NiagaraBool=true
     Locker Vector:Vector3f=<-10 0 0>
     Battle Royale Vector:Vector3f=<850 0 0>
     Rocket Racing Vector:Vector3f=<0 0 0>
     GameModeID:NiagaraInt32=$System.GameModeID
    )
    Offset Amount=1
    Ribbon Width=60
   )
   &AddVelocity(Velocity=<-1000 0 0>) {module=AddVelocity001}
  ]
  particleUpdateScript [
   &ParticleState()
   &InheritSourceMovement(
    Applied Position Scale=&Multiply_VectorByFloat(
     Vector=<1 1 1>
     Float=&FloatFromCurve(FloatCurve=DataInterface() {propertyValues {Curve {keys [{interpMode=RCIM_Cubic, tangentMode=RCTM_Break, time=0.5, value=1}, {interpMode=RCIM_Cubic, tangentMode=RCTM_Break, time=1, value=0}]}}})
    )
    ApplyForce=false
    ApplyPosition=true
    ApplyPositionFalloff=false
   )
   &SolveForcesAndVelocity(Clamp Velocity=false, Limit Acceleration=false)
   &VC_UpdateRibbon_v1(Enable Skeletal Position=true)
   &ScaleColor(
    Linear Color Curve=DataInterface() {propertyValues {
      AlphaCurve {keys [<at 0 0>,<at 0.1644 1.2>,<at 0.4621 1.1633>,<at 0.7108 0>]}
      BlueCurve {keys [<at 0.0385 0.1686>,<at 0.104 0.3333>,<at 0.1524 0.8854>,<at 0.2123 0.3333>]}
      GreenCurve {keys [<at 0.0385 0.0762>,<at 0.104 0.1412>,<at 0.1524 0.6313>,<at 0.2123 0.2261>]}
      RedCurve {keys [<at 0.0385 0.0351>,<at 0.104 0.1882>,<at 0.1524 0.6918>,<at 0.2123 0.2524>]}
     }}
    Scale Mode=::NewEnumerator2
   )
   &ScaleRibbonWidth(Ribbon Width Scale=&FloatFromCurve(FloatCurve=DataInterface() {propertyValues {Curve {keys [<at 0.1 1>,<at 1 0.8>]}}}))
   /SetVariables() {enabled=false, isSetParameters=true}
  ]
  rendererClasses [&NiagaraRibbonRendererProperties]
  renderers [
   {rendererClass=&NiagaraRibbonRendererProperties}
   {rendererClass=&NiagaraRibbonRendererProperties, rendererIndex=1}
   {rendererClass=&NiagaraRibbonRendererProperties, rendererIndex=2}
   {rendererClass=&NiagaraRibbonRendererProperties, rendererIndex=3}
   {rendererClass=&NiagaraRibbonRendererProperties, rendererIndex=4}
  ]
 }
 Ribbon003 {
  emitterUpdateScript [
   &EmitterState()
   &VC_FilteredSocketCount()
   &VC_RateQualityHandler(Epic Spawn Rate=35, High Spawn Rate=35, Low Spawn Rate=35, Medium Spawn Rate=35, Use Custom Rate?=true)
   &VC_SpawnRibbon_v1(Number of Ribbons=$Emitter.PortCount, SpawnRate=$Emitter.Spawn Rate)
  ]
  particleSpawnScript [
   &VC_PaintedVariantsOverride() {enabled=false}
   &VC_InitializeRibbon_v1(
    Color=<rgba 0.382 0.3201 0.8906 3>
    Custom Offset=true
    Enable Skeletal Position=true
    Lifetime=0.6
    Local Vector Offset=&VC_SetVectorByGameModeID(
     SameValueForLockerAndBR?:NiagaraBool=true
     Locker Vector:Vector3f=<10 0 0>
     Battle Royale Vector:Vector3f=<850 0 0>
     Rocket Racing Vector:Vector3f=<0 0 0>
     GameModeID:NiagaraInt32=$System.GameModeID
    )
    Offset Amount=1
    Ribbon Width=120
   )
   &AddVelocity(Velocity=<-4000 0 0>) {module=AddVelocity001}
   /SetVariables(Particles.InsideColor:LinearColor=<rgba 0.7348 0.678 0.9167 2>) {isSetParameters=true}
  ]
  particleUpdateScript [
   &ParticleState()
   &InheritSourceMovement(
    Applied Position Scale=&Multiply_VectorByFloat(
     Vector=<1 1 1>
     Float=&FloatFromCurve(FloatCurve=DataInterface() {propertyValues {Curve {keys [{interpMode=RCIM_Cubic, tangentMode=RCTM_Break, time=0.5, value=1}, {interpMode=RCIM_Cubic, tangentMode=RCTM_Break, time=1, value=0}]}}})
    )
    ApplyForce=false
    ApplyPosition=true
    ApplyPositionFalloff=false
   )
   &SolveForcesAndVelocity(Clamp Velocity=false, Limit Acceleration=false)
   &VC_UpdateRibbon_v1(Enable Skeletal Position=true)
   &ScaleColor(
    Scale Alpha=2
    Scale RGB=&VectorFromFloat(Value=&FloatFromCurve(FloatCurve=DataInterface() {propertyValues {Curve {
         defaultValue=1
         keys [
          {arriveTangent=13.4955, interpMode=RCIM_Cubic, leaveTangent=13.4955, tangentMode=RCTM_User, time=0, value=0.5}
          {arriveTangentWeight=0.1021, interpMode=RCIM_Cubic, leaveTangentWeight=0.3979, tangentMode=RCTM_User, time=0.1, value=1}
          {arriveTangent=0.0224, interpMode=RCIM_Cubic, leaveTangent=0.0224, tangentMode=RCTM_User, time=1, value=1}
         ]
        }}}))
    ScaleA=true
    ScaleRGB=true
   )
   &ScaleRibbonWidth(Ribbon Width Scale=&FloatFromCurve(FloatCurve=DataInterface() {propertyValues {Curve {keys [<at 0.1 1>,<at 1 0.8>]}}}))
   /SetVariables(
    Particles.R2 Position:NiagaraPosition=&AddVectorToPosition(Position=$Particles.Position, Vector=&TransformVector(Vector=<-30 0 0>, Destination Space=::Simulation))
    Particles.R2 Size:NiagaraFloat=&Multiply_Float(A=60, B=&FloatFromCurve(FloatCurve=DataInterface() {propertyValues {Curve {keys [<at 0 1>,<at 1 0>]}}}))
    Particles.R3:NiagaraFloat=&Multiply_Float(A=90, B=&FloatFromCurve(FloatCurve=DataInterface() {propertyValues {Curve {keys [<at 0 1>,<at 1 0>]}}}))
   ) {isSetParameters=true}
  ]
  rendererClasses [&NiagaraRibbonRendererProperties]
  renderers [
   {rendererClass=&NiagaraRibbonRendererProperties}
   {rendererClass=&NiagaraRibbonRendererProperties, rendererIndex=1}
   {rendererClass=&NiagaraRibbonRendererProperties, rendererIndex=2}
   {rendererClass=&NiagaraRibbonRendererProperties, rendererIndex=3}
  ]
 }
}
summary {
 systemName=NS_Boost_Chrome
 emitters [
  {emitterName=NE_VC_Boost_Sprite_Parent, rendererClasses [&NiagaraSpriteRendererProperties]}
  {emitterName=Ribbon002, rendererClasses [&NiagaraRibbonRendererProperties]}
  {emitterName=Ribbon003, rendererClasses [&NiagaraRibbonRendererProperties]}
 ]
 userVariables [
  User.bIsFrontendPreview {type=NiagaraBool}
  User.bIsPainted {type=NiagaraBool}
  User.ColorVariantID {type=NiagaraInt32}
  User.DTCarOffset {type=NiagaraPosition, description="", defaultValue {}}
  User.intensity_trail {type=NiagaraFloat, defaultValue=3, description=""}
  User.Lines_ribbon {type=NiagaraFloat, defaultValue=5, description=""}
  User.PaintedColor {type=LinearColor, description="", defaultValue {struct=LinearColor, value=<rgba 1 1 1 1>}}
  User.PaintedColor2 {type=LinearColor, description="", defaultValue {struct=LinearColor, value=<rgba 1 1 1 1>}}
  User.PaintedScalar {type=NiagaraFloat, defaultValue=1, description=""}
  User.PortCount {type=NiagaraInt32, defaultValue=1}
 ]
}
